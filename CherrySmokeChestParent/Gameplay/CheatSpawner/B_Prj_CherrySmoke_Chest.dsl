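; Thrown spawner projectile that places a Cherry Smoke chest building actor where it lands.
; On authority, HandleHit spawns only on valid, roughly upward-facing surfaces (normal-to-actor-up dot check), otherwise it may revert to a pickup.
; SpawnBGA_ adjusts for water and rotation, replaces an earlier chest on higher hits, and calls OnSpawnChest.
; It then fires a burst cue, FX and gameplay event, and destroys its owner.
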(event EventBeginPlay)

(event Collision|EventActorBeginOverlap
  (params (OtherActor :type Actor%20Object%20Reference)))

(event Collision|EventHit
  (params (MyComp :type Primitive%20Component%20Object%20Reference) (Other :type Actor%20Object%20Reference) (OtherComp :type Primitive%20Component%20Object%20Reference) (bSelfMoved :type Boolean) (HitLocation :type Vector :default "0, 0, 0") (HitNormal :type Vector :default "0, 0, 0") (NormalImpulse :type Vector :default "0, 0, 0") (Hit :type Hit%20Result%20Structure%20%28by%20ref%29))
  (|CallParentFunction MyComp Other OtherComp bSelfMoved HitLocation HitNormal NormalImpulse Hit)
  (Effects|Components|ParticleSystem|SpawnFXAtLocation (Variables|Default|GetFXSystem%20Asset) (Transformation|GetActorLocation)))

(fn GetOptionalOwner
  (bind _owneroverride (Variables|Default|GetOwnerOverride)
    (:then
      (return _owneroverride))
    (:else
      (return (Variables|Actor|GetInstigator)))))

(fn HandleHit
  (params (OtherActor :type Actor%20Object%20Reference) (OtherComp :type Primitive%20Component%20Object%20Reference) (Hit :type Hit%20Result%20Structure%20%28by%20ref%29) (HitLocation :type Vector :default "0, 0, 0"))
  (bind (_bblockinghit _binitialoverlap _time _distance _location _impactpoint _normal) (Collision|BreakHitResult Hit))
  (bind _returnvalue (Transformation|GetActorUpVector))
  (bind _returnvalue_2 (Math|Vector|DotProduct _normal _returnvalue))
  (Networking|SwitchHasAuthority
    (:Authority
      (bind (_validspawn _reverttopickup) (CallFunction|IsHitValidforSpawn OtherComp OtherActor Hit))
      (Development|PrintString (Utilities|String|ToString(Float) _returnvalue_2))
      (if (and _validspawn (>= _returnvalue_2 0.707))
        (CallFunction|InitiateSpawnBGA Hit OtherActor OtherComp HitLocation)
        (elif _reverttopickup
          (CallFunction|RevertThrownObjecttoPickup Hit))))))

(fn SpawnBGA_
  (params (Normal :type Vector :default "0, 0, 0") (HitActor :type Actor%20Object%20Reference) (HitComp :type Primitive%20Component%20Object%20Reference) (Pos :type Vector :default "0, 0, 0") (FromWater :type Boolean) (Hit :type Hit%20Result%20Structure))
  (bind (_returnvalue_x _returnvalue_y _returnvalue_z) (Math|Vector|BreakVector (Transformation|GetActorLocation (Variables|Spawning|GetSpawnedBGA))))
  (bind _returnvalue (Actor|GetOwner))
  (if (Variables|Spawning|GetSpawned)
    (if (> (.z Pos) _returnvalue_z)
      (Actor|DestroyActor (Variables|Spawning|GetSpawnedBGA)))
    (else
      (Variables|Spawning|SetSpawned true)
      (Utilities|FlowControl|Sequence
        (:then_0
          (Actor|SetLifeSpan (Variables|Spawning|GetLifespanAfterSpawnBGA))
          (Collision|SetActorEnableCollision false)
          (Rendering|SetActorHiddenInGame (Variables|Spawning|GetSetActorHiddenOnSpawnBGA)))
        (:then_1
          _))))
  (bind _returnvalue_2 (Building|SpawnBuildingActor (Variables|Default|GetChest%20BGA) (Math|Transform|MakeTransform (Math|Vector|vector+vector (CallFunction|CheckforWater Pos HitActor) (* "0, 0, 0" (Variables|Spawning|TransformData|GetSpawnPosOffsetAlongNormal)) (Math|Vector|MakeVector 0.0 0.0 (Variables|Spawning|TransformData|GetSpawnPosAdditionalZOffset))) (Math|Rotator|SelectRotator (Math|Rotator|MakeRotfromXZ (Math|Vector|ProjectVectorontoPlane (Transformation|GetActorForwardVector) (Math|Vector|SelectVector Normal "0, 0,1.000000" (Math|Vector|NotEqual(Vector) Normal "0, 0, 0" 0.0001))) (Math|Vector|SelectVector Normal "0, 0,1.000000" (Math|Vector|NotEqual(Vector) Normal "0, 0, 0" 0.0001))) (Transformation|GetActorRotation self) (Variables|Spawning|TransformData|GetUseImpactNormalForBGASpawnRotation)) (Variables|Spawning|TransformData|GetSpawnScale)) (CallFunction|GetOptionalOwner) (Variables|Actor|GetInstigator) false (Variables|Default|GetbShouldSpawnTransient)))
  (bind _output_get (Variables|Spawning|SetSpawnedBGA _returnvalue_2))
  (bind _asb_cherry_smoke_chest_container_base_parent (Utilities|Casting|CastToB_CherrySmokeChest_Container_BaseParent _output_get)) ; If this fails, ChestBGA isn't set.
  (Default|CallOnSpawnChest _asb_cherry_smoke_chest_container_base_parent)
  (GameplayCue|ExecuteGameplayCueOnActor(Burst) (Variables|Spawning|GetSpawnedBGA) (Variables|Default|GetImpactExplodeGC) (Ability|GameplayCue|MakeGameplayCueParameters :NormalizedMagnitude 0.0 :RawMagnitude 0.0 :Location (Transformation|GetActorLocation (Variables|Spawning|GetSpawnedBGA)) :Normal "0,0,0" :GameplayEffectLevel 1 :AbilityLevel 1 :bReplicateLocationWhenUsingMinimalRepProxy false))
  (Effects|Components|ParticleSystem|SpawnFXAtLocation (Variables|Default|GetFXSystem%20Asset) (Transformation|GetActorLocation))
  (Ability|SendGameplayEventtoActor (Actor|GetOwner) "(TagName=\"Event.Projectile.SpawnedBGA\")" (Utilities|Struct|MakeGameplayEventData :OptionalObject _returnvalue_2 :ContextHandle "()" :InstigatorTags "(GameplayTags=)" :TargetTags "(GameplayTags=)" :EventMagnitude 0.0 :TargetData "()"))
  (Utilities|IsValid _returnvalue
    (:Is%20Valid
      (Actor|DestroyActor _returnvalue)
      (return _returnvalue_2))))

(fn ConstructionScript
  (|CallParentFunction))
